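# Animation Blueprint for the F_MED stealth backpack, posing its root and wing bones from the wearer's motion.
# Owner spine_05 yaw relative to spine_01, as a ref-pose delta, drives an additive root roll and offset.
# Both wing bones blend to fixed mirrored rotations through an interpolated 2D-speed threshold, rising quickly and easing out slowly.
# Skydiving outside dive mode bypasses all these adjustments, blending to the cached input pose.
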AnimBlueprint /BRCosmetics/Accessories/FORT_Backpacks/F_MED_Stealth_Backpack/Meshes/F_MED_Stealth_Pack_AnimBp
target_skeleton=&/F_MED_Stealth_Pack_Skeleton
graphs [
 AnimGraph {nodes [
   n1 = "Transform (Modify) Bone - Bone: root"(
    ComponentPose=Local To Component(LocalPose=Use cached pose 'SavedPose30107'())
    Translation=Get backpack_translation()
    Rotation=Get backpack_rotate()
   ) {
    rotationMode=BMM_Additive
    rotationSpace=BCS_BoneSpace
    translationMode=BMM_Additive
    translationSpace=BCS_BoneSpace
    alphaBoolBlend {blendOption=Linear}
    boneToModify {boneName=root}
   }
   n2 = Get PawnSpeed2D()
   n3 = "Transform (Modify) Bone - Bone: dyn_wing_l"(ComponentPose=@n1, Alpha=@n2, Translation="0,0,0", Rotation="25,25,25", Scale="1,1,1") {
    rotationMode=BMM_Replace
    rotationSpace=BCS_BoneSpace
    alphaBoolBlend {blendOption=Linear}
    alphaScaleBiasClamp {bInterpResult=true, bMapRange=true, interpSpeedDecreasing=1, interpSpeedIncreasing=18, inRange {max=450, min=450}}
    boneToModify {boneName=dyn_wing_l}
    rotation=<rot 25 25 25>
   }
   n4 = "Transform (Modify) Bone - Bone: dyn_wing_r"(ComponentPose=@n3, Alpha=@n2, Translation="0,0,0", Rotation="25,-25,-25", Scale="1,1,1") {
    rotationMode=BMM_Replace
    rotationSpace=BCS_BoneSpace
    alphaBoolBlend {blendOption=Linear}
    alphaScaleBiasClamp {bInterpResult=true, bMapRange=true, interpSpeedDecreasing=1, interpSpeedIncreasing=18, inRange {max=450, min=450}}
    boneToModify {boneName=dyn_wing_r}
    rotation=<rot 25 -25 -25>
   }
   Output Pose(Result=Blend Poses by bool(
     BlendPose_0=Component To Local(ComponentPose=@n4)
     BlendPose_1=Use cached pose 'SavedPose30107'()
     bActiveValue=Get wings_bool() -> (wings_bool:bool)
    ) {blendType=Linear}) {layerGroup=None}
   AnimGraphNode_SubInput_0 = Input Pose()
   Save cached pose 'SavedPose30107'(Pose=@AnimGraphNode_SubInput_0) {cacheName=SavedPose30107}
  ]}
 EventGraph {nodes [
   Event BlueprintUpdateAnimation()
   > n1 = Sequence()
   > n2 = Branch(Condition=Get bIsSkyDiving())
   > n3 = Branch(Condition=Get bIsSkydiveDiveMode())
   > Set wings_bool(wings_bool:bool=true) -> (Output_Get:bool)
   Set wings_bool(execute=@n3.else, wings_bool=false)
   Set wings_bool(execute=@n2.else, wings_bool=true)
   n4 = GetOwnerSkeletalMeshComponent()
   n5 = BreakRotator(InRot=BreakTransform(InTransform=GetDeltaTransformFromRefPose(self=@n4, BoneName=spine_05, BaseName=spine_01)).Rotation)
   Set backpack_rotate(execute=@n1.then_1, backpack_rotate=MakeRotator(Roll=MapRangeClamped(Value=@n5.Yaw, InRangeA=0, InRangeB=-30, OutRangeB=20)))
   > Set backpack_translation(backpack_translation=MakeVector(Y=MapRangeClamped(Value=@n5.Yaw, InRangeA=0, InRangeB=-30, OutRangeB=-3)))
  ]}
]
variables [
 backpack_rotate {type=Rotator}
 backpack_translation {type=Vector}
 wings_bool {type=boolean, default=true}
]
